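; Passive inventory ability for the ChapWorm item that slowly grants kinetic energy to the weapon over time.
; On activation it starts a timer for IncrementKineticEnergy, with two tag-query wait tasks pausing it when the equip-or-overpower query turns true and unpausing it when false.
; Activation requires positive hotfixable gain interval and amount values.
; Ending the ability clears the timer and explicitly ends only the query-true task.
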(event Ability|EventActivateAbility
  (bind _equiporoverpowertagquery (Variables|Settings|GetEquipOrOverpowerTagQuery))
  ; ActivateAbility: Start a timer to slowly provide kinetic energy gain on the weapon when the item is in inventory.
  (|CallParentFunction)
  (Utilities|FlowControl|Sequence
    (:then_0
      (bind _returnvalue (Utilities|Time|SetTimerbyEvent (EventDispatchers|CreateEvent :!event IncrementKineticEnergy) (ScalableFloat|GetValueatLevel (Variables|Hotfix|GetHotfix_KineticEnergyPassiveInventoryGainInterval)) true true))
      (Variables|Transient|SetKineticEnergyPassiveGainTimerHandle _returnvalue))
    (:then_1
      (bind _asynctaskproxy (Ability|Tasks|WaitGameplayTagQuery _equiporoverpowertagquery)
        (:then
          (Variables|Transient|SetAsyncTask_WaitGameplyQueryTrue _asynctaskproxy)
          (bind _asynctaskproxy_2 (Ability|Tasks|WaitGameplayTagQuery _equiporoverpowertagquery 0 "WhenFalse")
            (:then
              (Variables|Transient|SetAsyncTask_WaitGameplayQueryFalse _asynctaskproxy_2))
            (:Triggered
              (Utilities|Time|UnpauseTimerbyHandle (Variables|Transient|GetKineticEnergyPassiveGainTimerHandle)))))
        (:Triggered
          (Utilities|Time|PauseTimerbyHandle (Variables|Transient|GetKineticEnergyPassiveGainTimerHandle)))))))

(event Ability|EventOnEndAbility
  (params (bWasCancelled :type Boolean))
  ;[0.1615 0.1154 0.01368 0.5] EndPlay: Clean up.
  (Utilities|FlowControl|Sequence
    (:then_0
      (Utilities|Time|ClearandInvalidateTimerbyHandle (Variables|Transient|GetKineticEnergyPassiveGainTimerHandle))
      (GameplayTasks|EndTask (Variables|Transient|GetAsyncTask_WaitGameplyQueryTrue)))
    (:then_1
      (|CallParentFunction bWasCancelled))))

(fn CanActivateAbility
  (params (ActorInfo :type Gameplay%20Ability%20Actor%20Info%20Structure) (Handle :type Gameplay%20Ability%20Spec%20Handle%20Structure))
  ; CanActivateAbility: Ability can only be activated if the passive inventory gain values are valid.
  (return 0 (and (> (ScalableFloat|GetValueatLevel (Variables|Hotfix|GetHotfix_KineticEnergyPassiveInventoryGainInterval)) 0) (> (ScalableFloat|GetValueasInteger (Variables|Hotfix|GetHotfix_KineticEnergyPassiveInventoryGainAmount)) 0))))
