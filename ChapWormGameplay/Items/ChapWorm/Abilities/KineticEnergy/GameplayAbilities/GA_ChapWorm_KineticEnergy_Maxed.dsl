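; GA_ChapWorm_KineticEnergy_Maxed is a gameplay ability in the ChapWorm item's kinetic energy ability set, named for the maxed energy state.
; On activation it calls the parent implementation, then starts the Overpower tutorial.
; It does this by sending the ChapWorm.Tutorial.Overpower.Activate.Start gameplay event to the avatar Fortnite player pawn.
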(event Ability|EventActivateAbility
  ;[0.03844 0.15 0.05151 0.5] ActivateAbility: Start the tutorial.
  (|CallParentFunction)
  (Ability|SendGameplayEventtoActor (CallFunction|GetAvatarasFortPlayerPawn) "(TagName=\"ChapWorm.Tutorial.Overpower.Activate.Start\")"))
