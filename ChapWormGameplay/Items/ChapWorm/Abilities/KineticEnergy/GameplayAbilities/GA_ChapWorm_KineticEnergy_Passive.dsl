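; Passive gameplay ability that charges the ChapWorm weapon's kinetic energy, stored as its ammo.
; A repeating timer adds charge; sprint tag-query tasks swap it for a sprint timer and a boosted-charging effect.
; Dealing damage with the weapon or taking enemy damage also adds charge, with a timed boost effect.
; Gain stops while the pawn has the Overpower tag; reaching 100 ammo applies a maxed effect.
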(event Ability|EventActivateAbility
  (bind _returnvalue (Ability|GetCurrentSourceWeapon))
  (bind _returnvalue_2 (Ability|GetCurrentSourceWeapon))
  (bind _sprintactivetagquery (Variables|Settings|GetSprintActiveTagQuery))
  ;[0.02272 0.09375 0.01087] ActivateAbility: Bind to know when sprint is active or inactive.
  (|CallParentFunction)
  (CallFunction|GetAvatarasValidFortPlayerPawn
    (:ValidFortPlayerPawn
      (Utilities|FlowControl|Sequence
        (:then_0
          ;;[0.02398 0.1771 0.1771] Set up async actions to know when the player is sprinting.
          (bind _asynctaskproxy (Ability|Tasks|WaitGameplayTagQuery _sprintactivetagquery)
            (:then
              (Variables|Transient|SetAsyncTask_WaitSprintTagQueryTrue _asynctaskproxy)
              (bind _asynctaskproxy_2 (Ability|Tasks|WaitGameplayTagQuery _sprintactivetagquery 0 "WhenFalse")
                (:then
                  (Variables|Transient|SetAsyncTask_WaitSprintTagQueryFalse _asynctaskproxy_2)
                  ;;[0.1423 0.04118 0.1615] Bind to know when the avatar player pawn does damage.
                  (EventDispatchers|BindEventtoOnDidDamage (CallFunction|GetAvatarasFortPlayerPawn) (EventDispatchers|CreateEvent :!event HandleAvatarPawnDidDamage))
                  ;;[0.007292 0.1378 0.1406] Timer to increment kinetic energy passively.
                  (bind _returnvalue_3 (Utilities|Time|SetTimerbyEvent (EventDispatchers|CreateEvent :!event HandlePassiveTimerTick) (ScalableFloat|GetValueatLevel (Variables|Hotfix|GetHotfix_KineticEnergyPassiveGainFrequnecy)) true true))
                  (Variables|Transient|SetPassiveGainTimerHandle _returnvalue_3)
                  ;;[0.08089 0.1666 0.224] Bind to know when the avatar player pawn is damaged.
                  (EventDispatchers|BindEventtoOnDamaged (CallFunction|GetAvatarasFortPlayerPawn) (EventDispatchers|CreateEvent :!event HandleAvatarPawnDamaged))
                  ;; Ensure there's atleast one ammo in the weapon.
                  (CallFunction|SetMinimumAmmoonWeapon)
                  ;[0.0237 0.1342 0.1458] Account for the ammo in the weapon currently.
                  (Utilities|IsValid _returnvalue_2
                    (:Is%20Valid
                      (CallFunction|HandleAmmoChanged (FortWeapon|GetLocalMagazineAmmoCount _returnvalue_2)))))
                (:Triggered
                  (CallFunction|HandleSprintActiveSet false))))
            (:Triggered
              (CallFunction|HandleSprintActiveSet true))))
        (:then_1
          ;;[0.04848 0.005947 0.05729] Bind to know when the ammo/charge for the weapon has changed.
          (Utilities|IsValid _returnvalue
            (:Is%20Valid
              (EventDispatchers|BindEventtoOnLocalAmmoChanged _returnvalue (EventDispatchers|CreateEvent :!event HandleAmmoChanged)))))))
    (:AvatarCastFailed
      (Ability|CancelAbility))))

(event Custom|HandleSprintActiveSet
  (params (bActive :type Boolean))
  ;[0.04013 0.1665 0.1771] HandleSprintActiveSet: Activate or deactive the timer that is responsible for ticking up the kinetic energy cause the avatar player pawn is sprinting.
  (if bActive
    (if (not (Utilities|Time|IsTimerActivebyHandle (Variables|Transient|GetKineticEnergySprintGainTimerHandle)))
      (bind _returnvalue (Utilities|Time|SetTimerbyEvent (EventDispatchers|CreateEvent :!event TickPassiveGainSprint) (ScalableFloat|GetValueatLevel (Variables|Hotfix|GetHotfix_KineticEnergySuperSprintGainFrequency)) true true))
      (Variables|Transient|SetKineticEnergySprintGainTimerHandle _returnvalue)
      (Utilities|Time|PauseTimerbyHandle (Variables|Transient|GetPassiveGainTimerHandle))
      (bind _returnvalue_2 (Ability|GameplayEffect|GetActiveGameplayEffectStackCount (Variables|Transient|GetBoostedChargingGameplayEffectHandle)))
      (if (<= _returnvalue_2 0)
        (bind _returnvalue_3 (Ability|ApplyGameplayEffectToOwner (Variables|Settings|GetBoostedChargingInfiniteGameplayEffectClass)))
        (Variables|Transient|SetBoostedChargingGameplayEffectHandle _returnvalue_3)))
    (else
      (Utilities|Time|ClearandInvalidateTimerbyHandle (Variables|Transient|GetKineticEnergySprintGainTimerHandle))
      (Utilities|Time|UnpauseTimerbyHandle (Variables|Transient|GetPassiveGainTimerHandle))
      (Ability|RemoveGameplayEffectFromOwnerWithHandle (Variables|Transient|GetBoostedChargingGameplayEffectHandle)))))

(event Ability|EventOnEndAbility
  (params (bWasCancelled :type Boolean))
  (bind _returnvalue (Ability|GetCurrentSourceWeapon))
  ;[0.09375 0.04926 0.00942] EndAbility: Clean up.
  (Utilities|FlowControl|Sequence
    (:then_0
      (CallFunction|HandleSprintActiveSet false)
      (Utilities|Time|ClearandInvalidateTimerbyHandle (Variables|Transient|GetPassiveGainTimerHandle))
      (GameplayTasks|EndTask (Variables|Transient|GetAsyncTask_WaitSprintTagQueryTrue))
      (Utilities|IsValid _returnvalue
        (:Is%20Valid
          (EventDispatchers|UnbindEventfromOnLocalAmmoChanged _returnvalue (EventDispatchers|CreateEvent :!event HandleAmmoChanged)))))
    (:then_1
      (bind _outfortplayerpawn (CallFunction|GetAvatarasValidFortPlayerPawn)
        (:ValidFortPlayerPawn
          (EventDispatchers|UnbindEventfromOnDidDamage _outfortplayerpawn (EventDispatchers|CreateEvent :!event HandleAvatarPawnDidDamage))
          (EventDispatchers|UnbindEventfromOnDamaged _outfortplayerpawn (EventDispatchers|CreateEvent :!event HandleAvatarPawnDamaged)))))
    (:then_2
      (|CallParentFunction bWasCancelled))))

(fn HandleAmmoChanged
  (params (LocalCount :type Integer) (LocalRemaining :type Integer))
  ;[0.08436 0.2161 0.2708] HandleAmmoChanged: When the kinetic energy (ammo) reaches 100, apply the Kinetic Energy Max gameplay effect.
  (if (== LocalCount 100)
    (bind _returnvalue (Ability|ApplyGameplayEffectToOwner (Variables|Settings|GetKineticEnergyMaxedGameplayEffectClass)))
    (Variables|Transient|SetKineticEnergyMaxedGameplayEffectHandle _returnvalue)
    (else
      (Ability|RemoveGameplayEffectFromOwnerWithHandle (Variables|Transient|GetKineticEnergyMaxedGameplayEffectHandle)))))

(fn HandleAvatarPawnDamaged
  (params (DamagedActor :type Actor%20Object%20Reference) (Damage :type Float%20%28single-precision%29) (InstigatedBy :type Controller%20Object%20Reference) (DamageCauser :type Actor%20Object%20Reference) (HitLocation :type Vector) (FHitComponent :type Primitive%20Component%20Object%20Reference) (BoneName :type Name) (Momentum :type Vector))
  (bind _hotfix_kineticenergytakedamagegain (Variables|Hotfix|GetHotfix_KineticEnergyTakeDamageGain))
  (bind _returnvalue (ScalableFloat|GetValueasInteger _hotfix_kineticenergytakedamagegain))
  (bind (_asfort_aicontroller _bsuccess) (Utilities|Casting|CastToFortAIController InstigatedBy))
  (bind (_asfort_player_controller _bsuccess_2) (Utilities|Casting|CastToFortPlayerController InstigatedBy))
  ;[0.03167 0.09304 0.09896] Increment kinetic energy when the avatar player pawn is damaged.
  (bind _asfort_player_pawn (Utilities|Casting|CastToFortPlayerPawn DamagedActor))
  (if (Math|Boolean|ANDBoolean (== _asfort_player_pawn (CallFunction|GetAvatarasFortPlayerPawn)) (or _bsuccess _bsuccess_2) (!= (Pawn|GetController (CallFunction|GetAvatarasFortPlayerPawn)) InstigatedBy) (not (Team|OnSameTeam (CallFunction|GetAvatarasFortPlayerPawn) InstigatedBy)))
    (if (> _returnvalue 0)
      (Ability|ApplyGameplayEffectToOwner (Variables|Settings|GetBoostedChargingDurationGameplayEffectClass))
      (CallFunction|IncrementKineticEnergy _returnvalue))))

(fn HandleAvatarPawnDidDamage
  (params (DamagedActor :type Actor%20Object%20Reference) (Damage :type Float%20%28single-precision%29) (InstigatedBy :type Controller%20Object%20Reference) (DamageCauser :type Actor%20Object%20Reference) (HitLocation :type Vector) (FHitComponent :type Primitive%20Component%20Object%20Reference) (BoneName :type Name) (Momentum :type Vector))
  (bind _hotfix_kineticenergydodamagegain (Variables|Hotfix|GetHotfix_KineticEnergyDoDamageGain))
  (bind _returnvalue (ScalableFloat|GetValueasInteger _hotfix_kineticenergydodamagegain))
  (bind (_asb_weapon_chap_worm _bsuccess) (Utilities|Casting|CastToB_Weapon_ChapWorm DamageCauser))
  ;[0.05029 0.1634 0.1667] HandleAvatrPawnDidDamage: When the player damages a fortplayerpawn, kinetic energy increments.
  (bind _asfort_pawn (Utilities|Casting|CastToFortPawn DamagedActor))
  (if (and (!= _asfort_pawn (CallFunction|GetAvatarasFortPlayerPawn)) _bsuccess)
    (if (> _returnvalue 0)
      (Ability|ApplyGameplayEffectToOwner (Variables|Settings|GetBoostedChargingDurationGameplayEffectClass))
      (CallFunction|IncrementKineticEnergy _returnvalue))))

(fn HandlePassiveTimerTick
  ;[0.07249 0.2297 0.2552] HandlePassiveTimerTick: Passive gain over time.
  (CallFunction|IncrementKineticEnergy (ScalableFloat|GetValueasInteger (Variables|Hotfix|GetHotfix_KineticEnergyPassiveGainAmount))))

(fn IncrementKineticEnergy
  (params (Amount :type Integer))
  (bind _returnvalue (CallFunction|GetAvatarasFortPlayerPawn))
  (bind _returnvalue_2 (Ability|GetCurrentSourceWeapon))
  ;[0.06803 0.1979 0.1979] IncrementKineticEnergy: Convert the input amount into ammo on the weapon.
  (Utilities|IsValid _returnvalue
    (:Is%20Valid
      (if (not (GameplayTags|HasMatchingGameplayTag _returnvalue "(TagName=\"ChapWorm.Overpower\")"))
        (Utilities|IsValid _returnvalue_2
          (:Is%20Valid
            (FortWeapon|ModAmmo _returnvalue_2 Amount)))))))

(fn SetMinimumAmmoOnWeapon
  (bind _returnvalue (Ability|GetCurrentSourceWeapon))
  ; Ensures there's minimum ammo count on the weapon.
  (Utilities|IsValid _returnvalue
    (:Is%20Valid
      (if (<= (FortWeapon|GetMagazineAmmoCount _returnvalue) 0)
        (FortWeapon|ModAmmo _returnvalue 1)))))

(fn TickPassiveGainSprint
  ;[0.1098 0.01263 0.1146] Kinetic Energy to be incremented when the player is sprinting.
  (CallFunction|IncrementKineticEnergy (ScalableFloat|GetValueasInteger (Variables|Hotfix|GetHotfix_KineticEnergySuperSprintGainAmount))))
